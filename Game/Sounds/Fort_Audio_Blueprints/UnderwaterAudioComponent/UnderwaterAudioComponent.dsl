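; Audio component that switches the 2D soundscape when the camera dips below or rises above the water surface.
; It binds to the WaterSubsystem's underwater state event and forces the surfaced state on BeginPlay and EndPlay.
; Submerging, skipped in local multiplayer, pushes a sound mix, raises a modulation control bus, plays a one-shot and fades in a reusable loop from a random offset.
; Surfacing reverses the mix and bus, fading out the loop with a stop sound.
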(event EventBeginPlay
  (Water|BindEventtoOnCameraUnderwaterStateChanged (WorldSubsystems|GetWaterSubsystem) (EventDispatchers|CreateEvent :!event Custom|On%20Camera%20Underwater%20State%20Changed))
  (CallFunction|OnCameraUnderwaterStateChanged false))

(event EventEndPlay
  (params (EndPlayReason :type EEndPlayReason%20Enum :default "Destroyed"))
  (CallFunction|OnCameraUnderwaterStateChanged false))

(event Custom|On%20Camera%20Underwater%20State%20Changed
  (params (bIsUnderWater :type Boolean) (DepthUnderwater :type Float%20%28single-precision%29))
  (bind _camunderwatersoundmix (Variables|Default|GetCamUnderwaterSoundMix))
  ;[1 1 1] Handle Camera dipping below water surface
  (if (Variables|Default|GetbEnabled)
    (if (Viewport|HasMultipleLocalPlayers)
      _
      (elif bIsUnderWater
        (Audio|PushSoundMixModifier _camunderwatersoundmix)
        (Audio|Modulation|SetGlobalControlBusMixValue "/Game/Sounds/Modulation/Parameters/CB_IsCameraUnderwater.CB_IsCameraUnderwater" 1.0 0.25)
        (Audio|PlaySound2D (Variables|Default|GetCamUnderwaterStartSound))
        (bind _camunderwateraudio (Variables|Default|GetCamUnderwaterAudio)
          (:then
            (if (Utilities|Enum|Equal(Enum) (Audio|Components|Audio|GetPlayState _camunderwateraudio) "FadingOut")
              (Audio|Components|Audio|FadeIn _camunderwateraudio 0.5 1.0 0.0 "Linear" :!class /Script/Engine.AudioComponent)))
          (:else
            (bind _returnvalue (Audio|SpawnSound2D (Variables|Default|GetCamUnderwaterLoopSound)))
            (bind _output_get (Variables|Default|SetCamUnderwaterAudio _returnvalue))
            (Audio|Components|Audio|FadeIn _output_get 0.5 1.0 (Math|Random|RandomFloatinRange 0.0 32.0) "Logarithmic" :!class /Script/Engine.AudioComponent)))))
    (Audio|PopSoundMixModifier _camunderwatersoundmix)
    (Audio|Modulation|SetGlobalControlBusMixValue "/Game/Sounds/Modulation/Parameters/CB_IsCameraUnderwater.CB_IsCameraUnderwater" 0.0 0.25)
    (bind _camunderwateraudio_2 (Variables|Default|GetCamUnderwaterAudio)
      (:then
        (Audio|Components|Audio|FadeOut _camunderwateraudio_2 0.5 0.0 "Linear" :!class /Script/Engine.AudioComponent)
        (Audio|PlaySound2D (Variables|Default|GetCamUnderwaterStopSound)))
      (:else))))
